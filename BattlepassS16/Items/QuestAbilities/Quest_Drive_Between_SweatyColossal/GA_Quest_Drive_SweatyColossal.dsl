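; Battle pass Season 16 quest gameplay ability that credits driving between two locations, apparently Sweaty Sands and Colossal Coliseum by its name.
; When event-activated on the server, it checks that the player's vehicle has the required tags and the player has the start-location tags, otherwise ending.
; It then waits in parallel for the destination tag, which sends the quest stat event and ends it, or a vehicle-exit event, which ends it without credit.
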(event Ability|EventActivateAbilityFromEvent
  (params (EventData :type Gameplay%20Event%20Data%20Structure%20%28by%20ref%29))
  ;[1 1 1] Activate
  (|CallParentFunction EventData)
  (if (Networking|IsServer)
    ;[1 1 1] Ensure player is in the correct vehicle and start location
    (bind _vehicle_occupied (Utilities|IsinVehicle (Variables|GAVars|Data|GetOwningPlayerPawn))
      (:true
        (if (and (GameplayTags|HasAllTags (Class|FortVehicleInterface|GetGameplayTags _vehicle_occupied) (Variables|ReadOnly|GetRequiredVehicleTags)) (GameplayTags|HasAllTags (GameplayTags|GetOwnedGameplayTags (Variables|GAVars|Data|GetOwningPlayerPawn) :!class /Script/GameplayTags.GameplayTagAssetInterface) (Variables|ReadOnly|GetStartLocationTags)))
          ;[1 1 1] Listen for destination tag or if player exits the vehicle
          (Ability|Tasks|WaitGameplayEvent "(TagName=\"Event.Vehicle.Exit\")" 0 true
            (:then
              (Ability|Tasks|WaitGameplayTagAdd (Variables|ReadOnly|GetDestinationLocationTag) 0 true
                (:Added
                  (Quest|SendComplexCustomStatEvent (Variables|GAVars|Data|GetOwningPlayerQuestManager) 0 (Variables|ReadOnly|GetQuestStatEventTags) "(GameplayTags=)" 1)
                  (Ability|EndAbility))))
            (:EventReceived
              (Development|PrintString "Player left vehicle; ending ability" true true "(R=1.000000,G=0.000000,B=0.266295,A=1.000000)")
              (Ability|EndAbility)))
          (else
            (Ability|EndAbility)))))))
